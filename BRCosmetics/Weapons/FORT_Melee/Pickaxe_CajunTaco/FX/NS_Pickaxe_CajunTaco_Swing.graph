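# Niagara swing effect for the CajunTaco pickaxe, drawing two short-lived ribbon animation trails, SkullTrail and UmbrellaTrail, between different bone or socket pairs.
# Both spawn through the FN AnimTrail modules and take tint from a system TOD value plus time-of-day color scaling.
# SkullTrail alpha cuts off by system age on a curve picked by SwingSectionIndex, with its TrailFade blend weighted to zero.
# UmbrellaTrail alpha follows TrailFade directly.
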NiagaraSystem /BRCosmetics/Weapons/FORT_Melee/Pickaxe_CajunTaco/FX/NS_Pickaxe_CajunTaco_Swing
dependencies {
 usedDataInterfaces [&NiagaraDataInterfaceColorCurve, &NiagaraDataInterfaceCurve, &NiagaraDataInterfaceSocketReader]
 usedDynamicInputs [&Add_Float]
 usedModules [
  &self:SystemSpawnScript.SystemScriptSource.SystemScriptGraph.NiagaraNodeAssignment_0.SetVariables_#1
  &SystemState
  &self:SystemSpawnScript.SystemScriptSource.SystemScriptGraph.NiagaraNodeAssignment_2.SetVariables_#2
 ]
}
emitters {
 SkullTrail {
  emitterUpdateScript [
   &EmitterState()
   &FN_SpawnAnimTrail(AnimTrailWidthMode=::FromSecond)
  ]
  particleSpawnScript [
   &FN_InitializeAnimTrailBoneIndecies(First Bone or Socket Index=1, Second Bone or Socket Index=2)
   &FN_InitializeAnimTrail(
    Color=&MultiplyLinearColor(
     Linear Color=<rgba 0.2643 0.3076 0.4375 1>
     Scale Factor=&MakeFloat4FromVectorAndFloat(
      "Vector (RGB)"=&MakeVectorFromLinearColorRGB(Color=$System.TOD)
      "Float (Alpha)"=&Lerp_Float(
       A=&MergeHelper_FloatSwitch(
        0=&FloatFromCurve(FloatCurve=DataInterface() {propertyValues {Curve {keys [<at 0.1 1>,<at 0.12 0>]}}}, CurveIndex=$Engine.System.Age)
        1=&FloatFromCurve(FloatCurve=DataInterface() {propertyValues {Curve {keys [<at 0.1 1>,<at 0.11 0>]}}}, CurveIndex=$Engine.System.Age)
        2=&FloatFromCurve(FloatCurve=DataInterface() {propertyValues {Curve {keys [<at 0.13 1>,<at 0.14 0>]}}}, CurveIndex=$Engine.System.Age)
        3=&FloatFromCurve(FloatCurve=DataInterface() {propertyValues {Curve {keys [<at 0.13 1>,<at 0.2 0>]}}}, CurveIndex=$Engine.System.Age)
        Index=$User.SwingSectionIndex
        Num=::NewEnumerator2
       )
       B=$System.TrailFade
       Alpha=0
      )
     )
    )
    Lifetime=0.12
   )
   &AddVelocity(Velocity=<0 400 0>) {enabled=false}
  ]
  particleUpdateScript [
   &SolveForcesAndVelocity(Clamp Velocity=false, Limit Acceleration=false) {enabled=false}
   &FN_UpdateAnimTrail() {module=FN_UpdateAnimTrail001}
   &ScaleColor(
    Scale Alpha=&FloatFromCurve(FloatCurve=DataInterface() {propertyValues {Curve {keys [<at 0.1 1>,<at 0.2 1>]}}}, CurveIndex=$Engine.System.Age)
    Scale RGB=&Multiply_Vector(
     A=&MakeVectorFromLinearColorRGB(Color=&ColorFromCurve(ColorCurve=DataInterface() {propertyValues {
         AlphaCurve {keys [<at 0 1 RCIM_Cubic>]}
         BlueCurve {keys [{arriveTangent=0.96, interpMode=RCIM_Cubic, time=0.1316, value=1}, {interpMode=RCIM_Cubic, leaveTangent=-0.1644, time=1, value=1}]}
         GreenCurve {keys [{arriveTangent=0.0419, interpMode=RCIM_Cubic, time=0.1316, value=1}, {interpMode=RCIM_Cubic, leaveTangent=0.2347, time=1, value=1}]}
         RedCurve {keys [{arriveTangent=-0.4464, interpMode=RCIM_Cubic, time=0.1316, value=1}, <at 1 1 RCIM_Cubic>]}
        }}))
     B=&MakeVectorFromLinearColorRGB(Color=&FN_Select_Color_from_Time_Of_Day(
       Use Manual Time of Day Thresholds=true
       Morning=<rgba 0.493 0.4792 0.7812 1>
       Day=<rgba 1 1 1 1>
       Evening=<rgba 0.5469 0.5284 0.7865 1>
       Night=<rgba 0.3581 0.3594 0.7812 1>
       Morning Threshold=9
       Day Threshold=17
       Evening Threshold=21
      ))
    )
    ScaleA=true
    ScaleRGB=true
   ) {module=ScaleColor001}
   /SetVariables() {isSetParameters=true}
   /SetVariables(Particles.OffsetedPosition:NiagaraPosition=$Particles.Position) {isSetParameters=true}
  ]
  rendererClasses [&NiagaraRibbonRendererProperties]
  renderers [{rendererClass=&NiagaraRibbonRendererProperties}]
 }
 UmbrellaTrail {
  emitterUpdateScript [
   &EmitterState()
   &FN_SpawnAnimTrail(AnimTrailWidthMode=::FromFirst, Spawn Interval=0) {module=FN_SpawnAnimTrail001}
  ]
  particleSpawnScript [
   &FN_InitializeAnimTrailBoneIndecies(Second Bone or Socket Index=2)
   &FN_InitializeAnimTrail(
    AnimTrailWidth=0.8
    Color=&MultiplyLinearColor(
     Linear Color=$System.TOD
     Scale Factor=&MakeFloat4FromVectorAndFloat("Vector (RGB)"=<1 1 1>, "Float (Alpha)"=$System.TrailFade)
    )
    Lifetime=0.12
   )
   &AddVelocity(Velocity=<0 400 0>) {enabled=false}
  ]
  particleUpdateScript [
   &SolveForcesAndVelocity(Clamp Velocity=false, Limit Acceleration=false) {enabled=false}
   &ScaleColor(
    Scale Alpha=&FloatFromCurve(FloatCurve=DataInterface() {propertyValues {Curve {keys [<at 0.3 1>,<at 1 1>]}}})
    Scale RGB=&Multiply_Vector(
     A=&MakeVectorFromLinearColorRGB(Color=&ColorFromCurve(ColorCurve=DataInterface() {propertyValues {
         AlphaCurve {keys [<at 0 1 RCIM_Cubic>]}
         BlueCurve {keys [<at 0.4018 1>]}
         GreenCurve {keys [<at 0.4018 1>]}
         RedCurve {keys [<at 0.4018 1>]}
        }}))
     B=&MakeVectorFromLinearColorRGB(Color=&FN_Select_Color_from_Time_Of_Day(
       Use Manual Time of Day Thresholds=true
       Morning=<rgba 0.493 0.4792 0.7812 1>
       Day=<rgba 1 1 1 1>
       Evening=<rgba 0.5469 0.5284 0.7865 1>
       Night=<rgba 0.3581 0.3594 0.7812 1>
       Morning Threshold=9
       Day Threshold=17
       Evening Threshold=21
      ))
    )
    ScaleA=true
    ScaleRGB=true
   )
   &FN_UpdateAnimTrail() {module=FN_UpdateAnimTrail001}
  ]
  rendererClasses [&NiagaraRibbonRendererProperties]
  renderers [{rendererClass=&NiagaraRibbonRendererProperties}]
 }
}
summary {
 systemName=NS_Pickaxe_CajunTaco_Swing
 emitters [
  {emitterName=UmbrellaTrail, rendererClasses [&NiagaraRibbonRendererProperties]}
  {emitterName=SkullTrail, rendererClasses [&NiagaraRibbonRendererProperties]}
 ]
 userVariables [
  User.bIsFrontEnd {type=NiagaraBool}
  User.bIsFrontEndPreview {type=NiagaraBool}
  User.RibbonIDOffset {type=NiagaraInt32}
  User.SwingSectionIndex {type=NiagaraInt32}
 ]
}
